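# Physics asset for the purple leather jacket backpack cosmetic, rigging a creature-like figure with head, arms, feet, ears and tail for secondary motion.
# Small capsules cover each bone, with an extra box on the back.
# Constraints chain root to back, which carries arms and head; the head carries ears and lower body, which carries feet and a seven-segment tail.
# Limit angles are tightest along the tail; back, arm and foot joints lock twist.
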PhysicsAsset /BRCosmetics/Accessories/FORT_Backpacks/Backpack_F_MED_Leather_Jacket_Purple/Meshes/F_MED_Leather_Jacket_Purple_Pack_Physics
bodies {
 dyn_arm_L_01 [{shapeType=Capsule, length=2, radius=2, shapeName=dyn_arm_L_01_capsule}]
 dyn_arm_L_02 [{shapeType=Capsule, length=2, radius=2, shapeName=dyn_arm_L_02_capsule}]
 dyn_arm_R_01 [{shapeType=Capsule, length=2, radius=2, shapeName=dyn_arm_R_01_capsule}]
 dyn_arm_R_02 [{shapeType=Capsule, length=2, radius=2, shapeName=dyn_arm_R_02_capsule}]
 dyn_back [
  {shapeType=Capsule, length=2, radius=2, shapeName=dyn_back_capsule, rotation=<rot -30.0001 0 0>}
  {shapeType=Box, extentX=1.5, extentY=8, extentZ=6, shapeName=dyn_back_box, center=<-5.979 0 10.4064>, rotation=<rot -30.0001 0 0>}
 ]
 dyn_ear_L_01 [{shapeType=Capsule, length=1, radius=1, shapeName=dyn_ear_L_01_capsule, rotation=<rot 90 0 0>}]
 dyn_ear_L_02 [{shapeType=Capsule, length=1, radius=1, shapeName=dyn_ear_L_02_capsule, rotation=<rot 90 0 0>}]
 dyn_ear_L_03 [{shapeType=Capsule, length=1, radius=1, shapeName=dyn_ear_L_03_capsule, rotation=<rot 90 0 0>}]
 dyn_ear_R_01 [{shapeType=Capsule, length=1, radius=1, shapeName=dyn_ear_R_01_capsule, rotation=<rot 90 0 0>}]
 dyn_ear_R_02 [{shapeType=Capsule, length=1, radius=1, shapeName=dyn_ear_R_02_capsule, rotation=<rot 90 0 0>}]
 dyn_ear_R_03 [{shapeType=Capsule, length=1, radius=1, shapeName=dyn_ear_R_03_capsule, rotation=<rot 90 0 0>}]
 dyn_foot_L [{shapeType=Capsule, length=2, radius=2, shapeName=dyn_foot_L_capsule}]
 dyn_foot_R [{shapeType=Capsule, length=2, radius=2, shapeName=dyn_foot_R_capsule}]
 dyn_head [{shapeType=Capsule, length=2, radius=2, shapeName=dyn_head_capsule, rotation=<rot -40.0002 0 0>}]
 dyn_lowerBody [{
   shapeType=Capsule
   length=0.25
   radius=4.75
   shapeName=dyn_lowerBody_capsule
   center=<0.4655 1.1537 -4.0595>
   rotation=<rot 11.0279 6.5739 -8.5033>
  }]
 dyn_tail_01 [{shapeType=Capsule, length=1, radius=1.5, shapeName=dyn_tail_01_capsule, rotation=<rot 90 0 0>}]
 dyn_tail_02 [{shapeType=Capsule, length=1, radius=1.5, shapeName=dyn_tail_02_capsule, rotation=<rot 90 0 0>}]
 dyn_tail_03 [{shapeType=Capsule, length=1, radius=1.5, shapeName=dyn_tail_03_capsule, rotation=<rot 90 0 0>}]
 dyn_tail_04 [{shapeType=Capsule, length=1, radius=1.5, shapeName=dyn_tail_04_capsule, rotation=<rot 90 0 0>}]
 dyn_tail_05 [{shapeType=Capsule, length=1, radius=1.5, shapeName=dyn_tail_05_capsule, rotation=<rot 90 0 0>}]
 dyn_tail_06 [{shapeType=Capsule, length=1, radius=1.5, shapeName=dyn_tail_06_capsule, rotation=<rot 90 0 0>}]
 dyn_tail_07 [{shapeType=Capsule, length=1, radius=1.5, shapeName=dyn_tail_07_capsule, rotation=<rot 90 0 0>}]
 root [{shapeType=Capsule, length=2, radius=2, shapeName=root_capsule}]
}
constraints [
 {bone1Name=dyn_back, bone2Name=root, swing1LimitDegrees=20, swing2LimitDegrees=20, twistLimitDegrees=45, twistMotion=Locked}
 {bone1Name=dyn_head, bone2Name=dyn_back, swing1LimitDegrees=15, swing2LimitDegrees=5, twistLimitDegrees=10}
 {bone1Name=dyn_ear_R_01, bone2Name=dyn_head, swing1LimitDegrees=10, swing2LimitDegrees=5, twistLimitDegrees=10}
 {bone1Name=dyn_ear_R_02, bone2Name=dyn_ear_R_01, swing1LimitDegrees=10, swing2LimitDegrees=5, twistLimitDegrees=10}
 {bone1Name=dyn_ear_R_03, bone2Name=dyn_ear_R_02, swing1LimitDegrees=10, swing2LimitDegrees=5, twistLimitDegrees=10}
 {bone1Name=dyn_ear_L_01, bone2Name=dyn_head, swing1LimitDegrees=10, swing2LimitDegrees=5, twistLimitDegrees=10}
 {bone1Name=dyn_ear_L_02, bone2Name=dyn_ear_L_01, swing1LimitDegrees=10, swing2LimitDegrees=5, twistLimitDegrees=10}
 {bone1Name=dyn_ear_L_03, bone2Name=dyn_ear_L_02, swing1LimitDegrees=10, swing2LimitDegrees=5, twistLimitDegrees=10}
 {bone1Name=dyn_lowerBody, bone2Name=dyn_head, swing1LimitDegrees=30, swing2LimitDegrees=10, twistLimitDegrees=5}
 {bone1Name=dyn_foot_R, bone2Name=dyn_lowerBody, swing1LimitDegrees=20, swing2LimitDegrees=20, twistLimitDegrees=45, twistMotion=Locked}
 {bone1Name=dyn_foot_L, bone2Name=dyn_lowerBody, swing1LimitDegrees=20, swing2LimitDegrees=20, twistLimitDegrees=45, twistMotion=Locked}
 {bone1Name=dyn_tail_01, bone2Name=dyn_lowerBody, swing1LimitDegrees=5, swing2LimitDegrees=5, twistLimitDegrees=10}
 {bone1Name=dyn_tail_02, bone2Name=dyn_tail_01, swing1LimitDegrees=5, swing2LimitDegrees=5, twistLimitDegrees=10}
 {bone1Name=dyn_tail_03, bone2Name=dyn_tail_02, swing1LimitDegrees=5, swing2LimitDegrees=5, twistLimitDegrees=5}
 {bone1Name=dyn_tail_04, bone2Name=dyn_tail_03, swing1LimitDegrees=5, swing2LimitDegrees=5, twistLimitDegrees=5}
 {bone1Name=dyn_tail_05, bone2Name=dyn_tail_04, swing1LimitDegrees=5, swing2LimitDegrees=5, twistLimitDegrees=5}
 {bone1Name=dyn_tail_06, bone2Name=dyn_tail_05, swing1LimitDegrees=5, swing2LimitDegrees=5, twistLimitDegrees=5}
 {bone1Name=dyn_tail_07, bone2Name=dyn_tail_06, swing1LimitDegrees=5, swing2LimitDegrees=5, twistLimitDegrees=5}
 {bone1Name=dyn_arm_R_01, bone2Name=dyn_back, swing1LimitDegrees=20, swing2LimitDegrees=20, twistLimitDegrees=45, twistMotion=Locked}
 {bone1Name=dyn_arm_R_02, bone2Name=dyn_arm_R_01, swing1LimitDegrees=20, swing2LimitDegrees=20, twistLimitDegrees=45, twistMotion=Locked}
 {bone1Name=dyn_arm_L_01, bone2Name=dyn_back, swing1LimitDegrees=20, swing2LimitDegrees=20, twistLimitDegrees=45, twistMotion=Locked}
 {bone1Name=dyn_arm_L_02, bone2Name=dyn_arm_L_01, swing1LimitDegrees=20, swing2LimitDegrees=20, twistLimitDegrees=45, twistMotion=Locked}
]
