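# Anim Blueprint for the purple leather jacket backpack, adding physics and corrective posing on its skeleton.
# Each update maps the owner mesh's spine_05-to-spine_01 delta yaw from ref pose into a clamped backpackRotate pitch.
# A RigidBody simulation in root-bone space feeds additive tail, pack and back bone offsets, with backpackRotate partially blended onto the pack.
# Another pack tilt reads bIsSkydiveFloating, which is absent from the listed variables.
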AnimBlueprint /BRCosmetics/Accessories/FORT_Backpacks/Backpack_F_MED_Leather_Jacket_Purple/Meshes/F_MED_Leather_Jacket_Purple_Pack_AnimBP
target_skeleton=&/F_MED_Leather_Jacket_Purple_Pack_Skeleton
graphs [
 AnimGraph {nodes [
   n1 = RigidBody(ComponentPose=Local To Component(LocalPose=Input Pose())) {
    bClampLinearTranslationLimitToRefPose=true
    bOverrideWorldGravity=true
    overridePhysicsAsset=&/F_MED_Leather_Jacket_Purple_Pack_Physics
    simulationSpace=BaseBoneSpace
    alphaBoolBlend {blendOption=Linear}
    baseBoneRef {boneName=root}
    componentAppliedLinearAccClamp=<3000 3000 3000>
    componentLinearAccScale=<0.25 0.25 0.25>
    overrideWorldGravity {z=-980}
   }
   n2 = "Transform (Modify) Bone - Bone: dyn_tail_01"(ComponentPose=@n1, Alpha=1, Translation="0,0,0", Rotation="10.736,9.326,14.540", Scale="1,1,1") {
    rotationMode=BMM_Additive
    rotationSpace=BCS_BoneSpace
    alphaBoolBlend {blendOption=Linear}
    boneToModify {boneName=dyn_tail_01}
    rotation=<rot 10.736 9.326 14.54>
   }
   n3 = "Transform (Modify) Bone - Bone: pack"(ComponentPose=@n2, Alpha=0.4, Translation="0,0,5", Rotation=Get backpackRotate() -> (backpackRotate:Rotator)) {
    rotationMode=BMM_Additive
    rotationSpace=BCS_BoneSpace
    translationMode=BMM_Additive
    alphaBoolBlend {blendOption=Linear}
    boneToModify {boneName=pack}
    translation {z=5}
   }
   n4 = "Transform (Modify) Bone - Bone: pack"!(ComponentPose=@n3, bAlphaBoolEnabled:bool=Get bIsSkydiveFloating(), Translation="-1,0,0", Rotation="20,0,0", Scale) -> (Pose) {
    alphaInputType=Bool
    rotationMode=BMM_Additive
    rotationSpace=BCS_BoneSpace
    translationSpace=BCS_BoneSpace
    alphaBoolBlend {blendInTime=1, blendOption=Linear, blendOutTime=1}
    boneToModify {boneName=pack}
    rotation {pitch=20}
    translation {x=-1}
   }
   n5 = "Transform (Modify) Bone - Bone: dyn_back"(ComponentPose=@n4, Alpha=1, Translation="0,1,0", Rotation="0,0,0", Scale="1,1,1") {translationMode=BMM_Additive, alphaBoolBlend {blendOption=Linear}, boneToModify {boneName=dyn_back}, translation {y=1}}
   Output Pose(Result=Component To Local(ComponentPose=@n5)) {layerGroup=None}
  ]}
 EventGraph {nodes [
   Event BlueprintUpdateAnimation()
   n1 = GetOwnerSkeletalMeshComponent()
   n2 = BreakRotator(InRot=BreakTransform(InTransform=GetDeltaTransformFromRefPose(self=@n1, BoneName=spine_05, BaseName=spine_01)).Rotation)
   > Set backpackRotate(backpackRotate:Rotator=MakeRotator(Pitch=MapRangeClamped(Value=@n2.Yaw, InRangeA=-5, InRangeB=10, OutRangeA=20, OutRangeB=0))) -> (Output_Get:Rotator)
   TryGetPawnOwner()
  ]}
]
variables [backpackRotate {type=Rotator}]
